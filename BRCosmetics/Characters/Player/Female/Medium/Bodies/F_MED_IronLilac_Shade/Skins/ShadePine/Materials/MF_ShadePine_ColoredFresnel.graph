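# Material function for the IronLilac ShadePine cosmetic that adds a two-color fresnel rim to incoming material attributes, toggled by static switch.
# Combined soft and sharp fresnel tints base color and adds emissive, with ColorA and ColorB split by view-space vertex normal against a direction.
# An optional spherical pass, applied on top, splits colors using offset, rotated, normalized pre-skinned local position in camera space, blended through a channel-selectable texture mask.
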MaterialFunction /BRCosmetics/Characters/Player/Female/Medium/Bodies/F_MED_IronLilac_Shade/Skins/ShadePine/Materials/MF_ShadePine_ColoredFresnel
graph {nodes [
  n1 = FunctionInput() {inputName=Material, inputType=FunctionInput_MaterialAttributes}
  n2 = GetMaterialAttributes(None=@n1) {attributeGetTypes [#1, #2]}
  n3 = VectorParameter() {group=ColoredFresnel, parameterName=ColorA, sortPriority=1, defaultValue=<rgba 0.1651 0.5457 0.9473 1>}
  n4 = NamedRerouteUsage()
  n5 = ScalarParameter() {defaultValue=0.6, group=ColoredFresnel, parameterName=SoftFresnelBias, sortPriority=4}
  n6 = DotProduct(A=CameraVectorWS(), B=PixelNormalWS())
  n7 = SmoothStep(Min=@n6, Value=ScalarParameter() {defaultValue=0.4, group=ColoredFresnel, parameterName=SharpFresnelBias, sortPriority=5}) {constValue=0.4}
  n8 = ScalarParameter() {defaultValue=6, group=ColoredFresnel, parameterName=SharpFresnelHardness, sortPriority=6}
  n9 = Max(A=Saturate(None=SmoothStep(Min=@n6, Value=@n5) {constMax=1.1, constValue=0.6}), B=Saturate(None=Multiply(A=@n7, B=@n8) {constB=6}))
  n10 = VectorParameter() {group=ColoredFresnel, parameterName=ColorB, sortPriority=1, defaultValue=<rgba 0.5333 0.1195 0.8796 1>}
  n11 = NamedRerouteUsage()
  n12 = Max(A=Multiply(A=@n3.RGB, B=Multiply(A=@n4, B=@n9)), B=Multiply(A=@n10.RGB, B=Multiply(A=@n9, B=@n11)))
  n13 = ScalarParameter() {defaultValue=1, group=ColoredFresnel, parameterName=EmissiveIntensity, sortPriority=2}
  n14 = SetMaterialAttributes!(
   MaterialAttributes=@n2.MaterialAttributes
   BaseColor=LinearInterpolate(A=@n2.BaseColor, B=Multiply(A=@n2.BaseColor, B=@n12), Alpha=@n9)
   EmissiveColor=Add(A=@n2.EmissiveColor, B=Multiply(A=@n12, B=@n13))
  ) {attributeSetTypes [#1, #2]}
  n15 = GetMaterialAttributes(None=@n14) {attributeGetTypes [#1, #2]}
  n16 = VectorParameter() {group=ColoredFresnel, parameterName=ColorA, sortPriority=1, defaultValue=<rgba 0.1651 0.5457 0.9473 1>}
  n17 = VectorParameter() {group=ColoredFresnel, parameterName=ColorB, sortPriority=1, defaultValue=<rgba 0.5333 0.1195 0.8796 1>}
  n18 = Max(A=Multiply(A=@n16.RGB, B=NamedRerouteUsage()), B=Multiply(A=@n17.RGB, B=NamedRerouteUsage()))
  n19 = TextureSampleParameter2D() {
   group=ColoredFresnel
   parameterName=SphericalFresnel_EmissiveMask
   samplerType=SAMPLERTYPE_Masks
   sortPriority=7
   texture=/BRCosmetics/Accessories/FORT_Backpacks/Backpack_F_MED_IronLilac/Skins/ShadePine/Textures/T_IronLilac_ShadePine_Backpack_FX2_PROTO
  }
  n20 = ChannelMaskParameter(None=@n19.RGB) {parameterName=SphericalFresnel_MaskChannel, sortPriority=8}
  n21 = ScalarParameter() {defaultValue=1, group=ColoredFresnel, parameterName=EmissiveIntensity, sortPriority=2}
  n22 = SetMaterialAttributes(
   MaterialAttributes=@n15.MaterialAttributes
   BaseColor=LinearInterpolate(A=@n15.BaseColor, B=LinearInterpolate(A=@n15.BaseColor, B=@n18, Alpha=@n20.R), Alpha=NamedRerouteUsage())
   EmissiveColor=LinearInterpolate(A=@n15.EmissiveColor, B=Multiply(A=@n18, B=@n21), Alpha=@n20.R)
  ) {attributeSetTypes [#1, #2]}
  FunctionOutput(None=StaticSwitchParameter(True=StaticSwitchParameter(True=@n22, False=@n14) {group=ColoredFresnel, parameterName=UseSphericalFresnel, sortPriority=0}, False=@n1) {defaultValue=true, group=ColoredFresnel, parameterName=UseColoredFresnel, sortPriority=0})
  n23 = DotProduct(
   A=VectorParameter() {group=ColoredFresnel, parameterName=ColorDirection, sortPriority=3, defaultValue=<rgba 1 0 0 1>}.RGB
   B=Transform(None=VertexNormalWS()) {transformSourceType=TRANSFORMSOURCE_World, transformType=TRANSFORM_View}
  )
  n24 = Saturate(None=@n23)
  n25 = Saturate(None=Multiply(A=@n23) {constB=-1})
  NamedRerouteDeclaration(None=Max(
    A=NamedRerouteDeclaration(None=@n24) {name=ColorAMask, nodeColor=<rgba 0.2891 1 0 1>}
    B=NamedRerouteDeclaration(None=@n25) {name=ColorBMask, nodeColor=<rgba 0.0312 0 1 1>}
   )) {name=CombinedMask, nodeColor=<rgba 0 0.5312 1 1>}
  n26 = VectorParameter() {group=ColoredFresnel, parameterName=SphericalFresnel_ColorDirection, sortPriority=9, defaultValue=<rgba 0 1 0 1>}
  n27 = LocalPosition() {includedOffsets=ExcludeOffsets}
  n28 = VectorParameter() {group=ColoredFresnel, parameterName=SphericalFresnel_PositionOffset, sortPriority=13, defaultValue=<rgba -0.325 0 -0.325 1>}
  n29 = RotateAboutAxis(
   NormalizedRotationAxis=VectorParameter() {parameterName=SphericalFresnel_RotationAxis, sortPriority=11, defaultValue=<rgba 0 1 0 1>}.RGB
   RotationAngle=ScalarParameter() {defaultValue=0.14, parameterName=SphericalFresnel_Rotation, sortPriority=12}
   PivotPoint=Constant3Vector()
   Position=Add(A=FunctionInput(Preview=@n27.XYZ) {bUsePreviewValueAsDefault=true, inputName=Pre-Skinned Local Pos}, B=@n28.RGB)
  )
  n30 = DotProduct(A=@n26.RGB, B=Transform(None=Normalize(VectorInput=@n29)) {transformSourceType=TRANSFORMSOURCE_World, transformType=TRANSFORM_Camera})
  n31 = Saturate(None=Multiply(A=@n30, B=ScalarParameter() {defaultValue=6, group=ColoredFresnel, parameterName=SphericalFresnel_Hardness, sortPriority=10}) {constB=6})
  n32 = Saturate(None=Multiply(
    A=Multiply(A=@n30) {constB=-1}
    B=ScalarParameter() {defaultValue=6, group=ColoredFresnel, parameterName=SphericalFresnel_Hardness, sortPriority=10}
   ) {constB=6})
  NamedRerouteDeclaration(None=Max(
    A=NamedRerouteDeclaration(None=@n31) {name=SphericalColorAMask, nodeColor=<rgba 1 0 0.3984 1>}
    B=NamedRerouteDeclaration(None=@n32) {name=SphericalColorBMask, nodeColor=<rgba 0 0.7187 1 1>}
   )) {name=SphericalCombinedMask, nodeColor=<rgba 1 0.9609 0 1>}
  Time()
 ]}
